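; Resizable Neon City sign frame prop that lights its sign face, optionally shutting it off by day, and feeds hit data to materials.
; Construction builds instanced corners and edges from two corner vectors, scales the sign, spreads optional struts, and applies material, Lumen and Switch fake-glow settings.
; Global Lumen fields act as one-shot controls, pushing boost to every instance then resetting.
; Creative bounds, origin and grid-guide overrides account for sign and struts.
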(event EventOnDamagePlayEffects
  (params (Damage :type Float%20%28single-precision%29) (DamageTags :type Gameplay%20Tag%20Container%20Structure%20%28by%20ref%29) (Momentum :type Vector :default "0, 0, 0") (HitInfo :type Hit%20Result%20Structure%20%28by%20ref%29) (InstigatedBy :type Fort%20Pawn%20Object%20Reference) (DamageCauser :type Actor%20Object%20Reference) (EffectContext :type Gameplay%20Effect%20Context%20Handle%20Structure))
  (bind (_origin _boxextent) (Collision|GetActorBounds true true))
  (bind (_bblockinghit _binitialoverlap _time _distance _location _impactpoint) (Collision|BreakHitResult HitInfo))
  (for _array_element (Actor|GetComponentsByClass "/Script/Engine.PrimitiveComponent")
    (Rendering|Material|SetCustomPrimitiveDataVector4 _array_element 7 (Math|Vector4|MakeVector4 (.x _impactpoint) (.y _impactpoint) (.z _impactpoint)))
    (Rendering|Material|SetCustomPrimitiveDataFloat _array_element 27 (/ 1.0 (Math|Float|Min(Float) (Math|Float|Max(Float) (.x _boxextent) (.y _boxextent)) (.z _boxextent))))))

(event EventBeginPlay
  (Utilities|FlowControl|Sequence
    (:then_0
      (Networking|CanExecuteCosmeticEvents
        (:True
          (CallFunction|SetupMaterialParamsonSignMesh))))
    (:then_1
      (Networking|SwitchHasAuthority
        (:Authority
          (EventDispatchers|BindEventtoOnDied (delegate Custom|CustomEvent)))))))

(event Custom|CustomEvent
  (params (DamagedActor :type Actor%20Object%20Reference) (Damage :type Float%20%28single-precision%29) (InstigatedBy :type Controller%20Object%20Reference) (DamageCauser :type Actor%20Object%20Reference) (HitLocation :type Vector) (FHitComponent :type Primitive%20Component%20Object%20Reference) (BoneName :type Name) (Momentum :type Vector))
  (Collision|SetCollisionEnabled (Variables|Default|GetFrame_B_Long_HISM) "NoCollision"))

(event EventOnDeathPlayEffects
  (params (Damage :type Float%20%28single-precision%29) (DamageTags :type Gameplay%20Tag%20Container%20Structure%20%28by%20ref%29) (Momentum :type Vector :default "0, 0, 0") (HitInfo :type Hit%20Result%20Structure%20%28by%20ref%29) (InstigatedBy :type Fort%20Pawn%20Object%20Reference) (DamageCauser :type Actor%20Object%20Reference) (EffectContext :type Gameplay%20Effect%20Context%20Handle%20Structure))
  (Audio|Components|Audio|FadeOut (Variables|Default|GetAudio) 0.0 0.0 "Linear" :!class /Script/Engine.AudioComponent))

(event EventOnDayPhaseChanged
  (params (CurrentDayPhase :type EFortDayPhase%20Enum :default "Morning") (PreviousDayPhase :type EFortDayPhase%20Enum :default "Morning") (bAtCreation :type Boolean))
  (if (Variables|Default|GetShuts%20Of%20During%20the%20Day)
    (Utilities|FlowControl|Delay (Math|Random|RandomFloatinRange 0.0 1e+01))
    (if (Utilities|Enum|Equal(Enum) CurrentDayPhase "Day")
      (Rendering|Material|SetScalarParameterValueonMaterials (Variables|Default|GetMainSignBox) "EmissiveMult" 0.0)
      (else
        _)))
  (Rendering|Material|SetScalarParameterValueonMaterials (Variables|Default|GetMainSignBox) "EmissiveMult" 1.0))

(event Custom|Update
  (bind _outactors (Actor|GetAllActorsOfClass "/Game/Environments/Asteria/Props/NeonCity/NeonCity_Signage/Blueprints/Asteria_NeonCity_SignFrame_E.Asteria_NeonCity_SignFrame_E_C"))
  (for _array_element _outactors
    (if (Variables|Default|GetShuts%20Of%20During%20the%20Day _array_element)
      (bind _returnvalue (Rendering|Material|GetScalarParameterValue "/Game/Packages/Fortress_SharedMaterials/GlobalMaterialParameters/FortniteMaterialParameters.FortniteMaterialParameters" "TimeOfDay_0-1" :!class /Script/Engine.KismetMaterialLibrary))
      (if (and (> 2e+01 (* _returnvalue 24.0)) (> (* _returnvalue 24.0) 8.0))
        (Rendering|Material|SetScalarParameterValueonMaterials (Variables|Default|GetMainSignBox) "EmissiveMult" 0.0)
        (else
          _)))
    (Rendering|Material|SetScalarParameterValueonMaterials (Variables|Default|GetMainSignBox) "EmissiveMult" 1.0)))

(fn AddInstances_Corner
  (Components|InstancedStaticMesh|AddInstances (Variables|Default|GetFrame_A_Corner_HISM) (Utilities|Array|MakeArray (Math|Transform|MakeTransform (- (Variables|Default|GetTopLeftFlattened) "-128.000000,0.000000,0.000000") "90.000000,0.000000,0.000000") (Math|Transform|MakeTransform (+ (Variables|Default|GetBottomRightFlattened) "-128.000000,0.000000,0.000000") "-90.000000,0.000000,0.000000") (Math|Transform|MakeTransform (- (Math|Vector|MakeVector (.x (Variables|Default|GetBottomRightFlattened)) 0.0 (.z (Variables|Default|GetTopLeftFlattened))) "0.000000,0.000000,128.000000")) (Math|Transform|MakeTransform (+ (Math|Vector|MakeVector (.x (Variables|Default|GetTopLeftFlattened)) 0.0 (.z (Variables|Default|GetBottomRightFlattened))) "0.000000,0.000000,128.000000") "180.000000,0.000000,0.000000")) false false true))

(fn AddInstances_Edges
  (Components|InstancedStaticMesh|AddInstances (Variables|Default|GetFrame_B_Long_HISM) (Utilities|Array|MakeArray (Math|Transform|MakeTransform (- (Variables|Default|GetTopLeftFlattened) "0.000000,0.000000,128.000000") "180.000000,0.000000,0.000000" (Math|Vector|MakeVector 1.0 1.0 (* (/ (- (- (.z (Variables|Default|GetTopLeftFlattened)) (.z (Variables|Default|GetBottomRightFlattened))) 256.0) 512.0) 1.0))) (Math|Transform|MakeTransform (- (Math|Vector|MakeVector (.x (Variables|Default|GetBottomRight)) 0.0 (.z (Variables|Default|GetBottomRight))) "0.000000,0.000000,-128.000000") "0, 0, 0" (Math|Vector|MakeVector 1.0 1.0 (* (/ (- (- (.z (Variables|Default|GetTopLeftFlattened)) (.z (Variables|Default|GetBottomRightFlattened))) 256.0) 512.0) 1.0))) (Math|Transform|MakeTransform (+ (+ (Math|Vector|MakeVector (.x (Variables|Default|GetBottomRight)) 0.0 (.z (Variables|Default|GetTopLeftFlattened))) "-128.000000,0.000000,0.000000") "0.000000,0.000000,-0.005000") "90.000000,0.000000,0.000000" (* (Math|Vector|MakeVector 1.0 1.0 (/ (- (- (.x (Variables|Default|GetBottomRightFlattened)) (.x (Variables|Default|GetTopLeftFlattened))) 256.0) 512.0)) "0.995000,0.995000,1.000000")) (Math|Transform|MakeTransform (+ (+ (Math|Vector|MakeVector (.x (Variables|Default|GetTopLeftFlattened)) 0.0 (.z (Variables|Default|GetBottomRight))) "128.000000,0.000000,0.000000") "0.000000,0.000000,0.005000") "-90.000000,0.000000,0.000000" (Math|Vector|MakeVector 0.995 0.995 (/ (- (- (.x (Variables|Default|GetBottomRightFlattened)) (.x (Variables|Default|GetTopLeftFlattened))) 256.0) 512.0)))) false false true))

(fn AddStruts
  (bind _strutcount (Variables|Default|GetStrutCount))
  (Utilities|Array|Clear (Variables|Default|GetStrutMeshComponents))
  (if (Variables|Default|GetbAddStruts)
    (for _index (range _strutcount)
      (bind _returnvalue (AddComponent|NODE_AddStaticMeshComponent-0 :RelativeTransform (Math|Transform|MakeTransform (Math|Vector|MakeVector (- (.x (Variables|Default|GetTopLeft)) 42.0) 0.0 (+ (- (.z (Variables|Default|GetTopLeft)) (* (/ (+ (* (Math|Float|Sign(Float) (* (- (/ (Math|Conversions|ToFloat(Integer) (+ _index 1)) (Math|Conversions|ToFloat(Integer) (+ _strutcount 1))) 0.5) 2.0)) (- 1.0 (Math|Float|Power (- 1.0 (Math|Float|Absolute(Float) (* (- (/ (Math|Conversions|ToFloat(Integer) (+ _index 1)) (Math|Conversions|ToFloat(Integer) (+ _strutcount 1))) 0.5) 2.0))) (Variables|Default|GetStrutSpreadFactor)))) 1.0) 2.0) (- (.z (Variables|Default|GetTopLeft)) (.z (Variables|Default|GetBottomRight))))) 39.5)) "180.000000,0.000000,0.000000")))
      (Utilities|Array|Add (Variables|Default|GetStrutMeshComponents) _returnvalue)
      (if (> (Variables|Default|GetStrutCount) 1)
        (Components|StaticMesh|SetStaticMesh _returnvalue "/Game/Environments/Asteria/Props/NeonCity/NeonCity_Signage/Meshes/SM_NeonCity_SignFrame_Holder_B.SM_NeonCity_SignFrame_Holder_B")
        (else
          (Components|StaticMesh|SetStaticMesh _returnvalue "/Game/Environments/Asteria/Props/NeonCity/NeonCity_Signage/Meshes/SM_NeonCity_SignFrame_Holder_A.SM_NeonCity_SignFrame_Holder_A"))))))

(fn BuildExposedAssetProviderVersion(Implementation)
  (params (Version :type Fort%20Exposed%20Asset%20Provider%20Version%20Structure%20%28by%20ref%29))
  (|CallParentFunction Version)
  (ExposedAssetProvider|AddVersion Version 1)
  (return))

(fn CalculateMainSignBounds
  (bind (_top_left_worldspace _top_right_worldspace _bottom_left_worldspace _bottom_right_worldspace) (CallFunction|GetSignCornersWorldspace))
  (bind (_returnvalue_x _returnvalue_y) (Math|Vector|BreakVector (Transformation|GetWorldScale (Variables|Default|GetMainSignBox))))
  (return (Math|Vector|VectorComponentMax (Math|Vector|VectorGetAbs (- (Math|Vector|VectorComponentMax (Math|Vector|VectorComponentMax _top_left_worldspace _top_right_worldspace) (Math|Vector|VectorComponentMax _bottom_left_worldspace _bottom_right_worldspace)) (Transformation|GetWorldLocation (Variables|Default|GetMainSignBox)))) (* (LocalVariables|GetMinimumExtent) _returnvalue_y))))

(fn CalculateOverallBounds
  (LocalVariables|SetCurrentOrigin (Transformation|GetWorldLocation (Variables|Default|GetMainSignBox)))
  (LocalVariables|SetCurrentExtent (Default|CalculateMainSignBounds))
  ;[0.3646 1 0.1309 0.5] Combine the bounds of the sign box with the bounds of each strut on the sign
  (for _array_element (Variables|Default|GetStrutMeshComponents)
    (bind (_combinedorigin _combinedboxextent) (CallFunction|CombineBounds (Collision|GetComponentBounds _array_element) (.BoxExtent (Collision|GetComponentBounds _array_element)) (LocalVariables|GetCurrentOrigin) (LocalVariables|GetCurrentExtent)))
    (LocalVariables|SetCurrentOrigin _combinedorigin)
    (LocalVariables|SetCurrentExtent _combinedboxextent))
  (return (LocalVariables|GetCurrentOrigin) (LocalVariables|GetCurrentExtent)))

(fn CollectAssetReferencestoAllow(Implementation)
  (params (Collector :type Fort%20Exposed%20Asset%20Reference%20Collector%20Structure%20%28by%20ref%29))
  (|CallParentFunction Collector)
  (ExposedAssetProvider|CollectObjectPropertyAssetReferences Collector 0 (Variables|Default|GetMaterial))
  (ExposedAssetProvider|CollectObjectPropertyAssetReferences Collector 0 (Variables|Default|GetDummyMaterial))
  (return))

(fn CombineBounds
  (params (OriginA :type Vector) (BoxExtentA :type Vector) (OriginB :type Vector) (BoxExtentB :type Vector))
  (bind _originb (Variables|GetOriginB))
  (bind _boxextentb (Variables|GetBoxExtentB))
  (bind _origina (Variables|GetOriginA))
  (bind _boxextenta (Variables|GetBoxExtentA))
  ;[0.4769 1 0.8897 0.5] Get the maximum and minimum for each bounding box
  (LocalVariables|SetMaxBoundsA (+ _origina _boxextenta))
  (LocalVariables|SetMinBoundsA (- _origina _boxextenta))
  (LocalVariables|SetMaxBoundsB (+ _originb _boxextentb))
  (LocalVariables|SetMinBoundsB (- _originb _boxextentb))
  ;[0.2758 1 0.4005 0.5] Get the overall maximums and minimums of those values
  (LocalVariables|SetMaxOverall (Math|Vector|VectorComponentMax (LocalVariables|GetMaxBoundsA) (LocalVariables|GetMaxBoundsB)))
  (LocalVariables|SetMinOverall (Math|Vector|VectorComponentMin (LocalVariables|GetMinBoundsA) (LocalVariables|GetMinBoundsB)))
  ;[0.2814 1 0.1527 0.5] Calculate the overall midpoint (new origin) and the extent from it (new box extent)
  (return (/ (+ (LocalVariables|GetMaxOverall) (LocalVariables|GetMinOverall)) 2) (- (/ (+ (LocalVariables|GetMaxOverall) (LocalVariables|GetMinOverall)) 2) (LocalVariables|GetMinOverall))))

(fn GetCollisionStaticMesh
  (bind _mainsignbox (Variables|Default|GetMainSignBox))
  (bind _staticmesh (Variables|/Script/Engine.StaticMeshComponent|GetStaticMesh _mainsignbox))
  (bind (_returnvalue_min _returnvalue_max) (Utilities|Struct|BreakBox (StaticMesh|GetBoundingBox _staticmesh)))
  (bind (_returnvalue_location _returnvalue_rotation) (Math|Transform|BreakTransform (Transformation|GetActorTransform)))
  (LocalVariables|SetBoundingMin _returnvalue_min)
  (LocalVariables|SetBoundingMax _returnvalue_max)
  (return (Utilities|Operators|Subtract (+ (/ (- (Math|Transform|TransformLocation (Math|Transform|ComposeTransforms (Transformation|GetActorTransform) (Math|Transform|InvertTransform (Math|Transform|MakeTransform _returnvalue_location _returnvalue_rotation))) (LocalVariables|GetBoundingMax)) (Math|Transform|TransformLocation (Math|Transform|ComposeTransforms (Transformation|GetActorTransform) (Math|Transform|InvertTransform (Math|Transform|MakeTransform _returnvalue_location _returnvalue_rotation))) (LocalVariables|GetBoundingMin))) 2) (Utilities|MainSignBoxOffset)) (Math|Vector|MakeVector (.x (/ (- (LocalVariables|GetBoundingMax) (LocalVariables|GetBoundingMin)) 2)) (.y (/ (- (LocalVariables|GetBoundingMax) (LocalVariables|GetBoundingMin)) 2)) (* (- (.z (Variables|Default|GetTopLeftFlattened)) (.z (Variables|Default|GetBottomRightFlattened))) 0.5))) (Variables|/Script/Engine.StaticMeshComponent|GetStaticMesh (Variables|Default|GetMainSignBox))))

(fn GetCreativeActorBounds
  (bind (_overallorigin_ws _overallbounds) (CallFunction|CalculateOverallBounds))
  (return true _overallbounds))

(fn GetCreativeActorOrigin
  (if (or (not (Variables|Default|GetbAddStruts)) (== (Variables|Default|GetStrutCount) 0))
    (return true (Utilities|MainSignBoxOffset))
    (else
      ;[0.4958 1 0.1912 0.5] Convert the overall origin location from worldspace to local (relative to the sign mesh)
      (bind _overallorigin_ws (CallFunction|CalculateOverallBounds))
      (bind _returnvalue_location (Math|Transform|BreakTransform (Math|Transform|MakeRelativeTransform (Math|Transform|MakeTransform _overallorigin_ws) (Transformation|GetWorldTransform (Variables|Default|GetMainSignBox)))))
      (return true _returnvalue_location))))

(fn GetCreativeActorRelativeTransform
  (return false 0))

(fn GetGridAxisGuideAttachmentPoint
  (if (or (not (Variables|Default|GetbAddStruts)) (== (Variables|Default|GetStrutCount) 0))
    ;[1 0.6799 0.2575 0.5] Invert the offset relative to the origin to realign the grid axis guide
    (return true true (- (* (Variables|/Script/Engine.SceneComponent|GetRelativeLocation (Variables|Default|GetMainSignBox)) -1) "0.000000,0.000000,2.000000"))
    (else
      ;;[0.4958 1 0.1912 0.5] Convert the overall origin location to be local (relative to the sign mesh)
      (bind _overallorigin_ws (CallFunction|CalculateOverallBounds))
      (bind _returnvalue_location (Math|Transform|BreakTransform (Math|Transform|MakeRelativeTransform (Math|Transform|MakeTransform _overallorigin_ws) (Transformation|GetWorldTransform (Variables|Default|GetMainSignBox)))))
      (return true true (- (* _returnvalue_location -1) "0.000000,0.000000,2.000000")))))

(fn GetMeshComponents
  (LocalVariables|SetAllMeshComponents (Actor|GetComponentsByClass "/Script/Engine.MeshComponent"))
  (Utilities|Array|AppendArray (LocalVariables|GetAllMeshComponents) (Variables|Default|GetStrutMeshComponents))
  ;[0.5866 1 0.4762 0.5] The move tool's grid axis guide attaches to the first component in this array. Ensuring that's the sign box is the easiest way for us to deal with it in GetGridAxisGuideAttachmentPoint
  (Utilities|Array|SwapArrayElements (LocalVariables|GetAllMeshComponents) (Utilities|Array|FindItem (LocalVariables|GetAllMeshComponents) (Variables|Default|GetMainSignBox)))
  (return (LocalVariables|GetAllMeshComponents)))

(fn GetSignCornersWorldspace
  (bind _mainsignbox (Variables|Default|GetMainSignBox))
  (bind _returnvalue (Transformation|GetWorldLocation _mainsignbox))
  (bind _mainsignbox_2 (Variables|Default|GetMainSignBox))
  (bind _returnvalue_2 (Transformation|GetWorldLocation _mainsignbox_2))
  (bind _offset_topleft_bottomright (LocalVariables|GetOffset_TopLeft_BottomRight))
  (bind _offset_bottomleft_topright (LocalVariables|GetOffset_BottomLeft_TopRight))
  (bind _topleft_x (Math|Vector|BreakVector (Variables|GetTopLeft_X)))
  (bind (_topleft_x_2 _topleft_y _topleft_z) (Math|Vector|BreakVector (Variables|GetTopLeft_X)))
  (bind (_bottomright_x _bottomright_y _bottomright_z) (Math|Vector|BreakVector (Variables|GetBottomRight_X)))
  ;[0.5375 1 0.2545 0.5] Get the world location of each corner
  (LocalVariables|SetOffset_TopLeft_BottomRight (Math|Vector|RotateVector (- (Math|Vector|MakeVector _topleft_x_2 0.0 _topleft_z) (Utilities|MainSignBoxOffset)) (Transformation|GetActorRotation)))
  (LocalVariables|SetCorner_TL (+ _returnvalue_2 _offset_topleft_bottomright))
  (LocalVariables|SetCorner_BR (- _returnvalue_2 _offset_topleft_bottomright))
  (LocalVariables|SetOffset_BottomLeft_TopRight (Math|Vector|RotateVector (Utilities|Operators|Subtract (Math|Vector|MakeVector _topleft_x 0.0 _bottomright_z) (Utilities|MainSignBoxOffset)) (Transformation|GetActorRotation)))
  (LocalVariables|SetCorner_BL (+ _returnvalue _offset_bottomleft_topright))
  (LocalVariables|SetCorner_TR (- _returnvalue _offset_bottomleft_topright))
  (return (LocalVariables|GetCorner_TL) (LocalVariables|GetCorner_TR) (LocalVariables|GetCorner_BL) (LocalVariables|GetCorner_BR)))

(fn IgnoreWorldGridScaleOverride
  (return false))

(fn InitializeVectorParams
  (Variables|Default|SetBottomRightFlattened (* (Variables|Default|GetBottomRight) "1.000000,0.000000,1.000000"))
  (Variables|Default|SetTopLeftFlattened (* (Variables|Default|GetTopLeft) "1.000000,0.000000,1.000000")))

(fn SetGlowVisibility
  (bind _glow_color (Variables|FakeGlow|GetGlow_Color))
  (bind _returnvalue (Math|Conversions|ToVector(LinearColor) _glow_color))
  (bind _switch_fakeglow (Variables|Default|GetSwitch_FakeGlow))
  (Rendering|SetVisibility (Variables|Default|GetSwitch_FakeGlow) (and (Variables|FakeGlow|GetbEnableLightGlow) (TouchPlatform|IsNintendoSwitchorNintendoSwitchPreview)) false)
  (Rendering|Material|SetCustomPrimitiveDataVector3 _switch_fakeglow 0 _returnvalue)
  (Rendering|Material|SetDefaultCustomPrimitiveDataVector3 _switch_fakeglow 0 _returnvalue))

(fn SetupMaterialParamsOnSignMesh
  (bind _mainsignbox (Variables|Default|GetMainSignBox))
  (bind _pixelcellsize (Variables|Default|GetPIxelCellSize))
  (Rendering|Material|SetDefaultCustomPrimitiveDataFloat _mainsignbox 0 _pixelcellsize)
  (Rendering|Material|SetCustomPrimitiveDataFloat _mainsignbox 0 _pixelcellsize)
  (Rendering|Material|SetScalarParameterValueonMaterials (Variables|Default|GetMainSignBox) "LumenBoost" (Variables|Lumen|GetLumenBoost))
  (Rendering|Material|SetScalarParameterValueonMaterials (Variables|Default|GetMainSignBox) "LumenUseColorOverride" (Math|Float|Clamp(Float) (Variables|Lumen|GetLumenUseColorOverride)))
  (Rendering|Material|SetVectorParameterValueonMaterials (Variables|Default|GetMainSignBox) "LumenColorOverride" (Math|Conversions|ToVector(LinearColor) (Variables|Lumen|GetLumenColorOverride)))
  (Rendering|Material|SetScalarParameterValueonMaterials (Variables|Default|GetMainSignBox) "LumenSaturation" (Math|Float|Min(Float) (- 1.0 (Variables|Lumen|GetLumenSaturation)) 1.0)))

(fn ConstructionScript
  (Utilities|FlowControl|Sequence
    (:then_0
      (Utilities|FlowControl|Sequence
        (:then_0
          (if (!= (Variables|LumenGlobal|GetAddLumenBoost%20all%20Instances) 0)
            (bind _outactors (Actor|GetAllActorsOfClass "/Game/Environments/Asteria/Props/NeonCity/NeonCity_Signage/Blueprints/Asteria_NeonCity_SignFrame_E.Asteria_NeonCity_SignFrame_E_C"))
            (for _array_element _outactors
              (Variables|Lumen|SetLumenBoost (+ (Variables|Lumen|GetLumenBoost) (Variables|LumenGlobal|GetAddLumenBoost%20all%20Instances)) _array_element)
              (CallFunction|SetupMaterialParamsonSignMesh _array_element))
            (Variables|LumenGlobal|SetAddLumenBoost%20all%20Instances 0.0)))
        (:then_1
          (if (>= (Variables|LumenGlobal|GetForceAllSignsLumenBoost%20all%20Instances) 0)
            (bind _outactors_2 (Actor|GetAllActorsOfClass "/Game/Environments/Asteria/Props/NeonCity/NeonCity_Signage/Blueprints/Asteria_NeonCity_SignFrame_E.Asteria_NeonCity_SignFrame_E_C"))
            (for _array_element_2 _outactors_2
              (Variables|Lumen|SetLumenBoost (Variables|LumenGlobal|GetForceAllSignsLumenBoost%20all%20Instances) _array_element_2)
              (CallFunction|SetupMaterialParamsonSignMesh _array_element_2))
            (Variables|LumenGlobal|SetForceAllSignsLumenBoost%20all%20Instances -1.0)))))
    (:then_1
      (Utilities|FlowControl|Sequence
        (:then_0
          (CallFunction|InitializeVectorParams))
        (:then_1
          (CallFunction|AddInstancesCorner)
          (CallFunction|AddInstancesEdges))
        (:then_2
          (CallFunction|ValidateSignPlaneTransform))
        (:then_3
          (CallFunction|AddStruts))
        (:then_4
          (CallFunction|ValidateMaterialOverride)
          (CallFunction|ValidateMaterialTwoSided) ; Do this before setting params so the back of a two-sided sign is affected
          (CallFunction|SetupMaterialParamsonSignMesh))
        (:then_5
          (CallFunction|SetGlowVisibility))))))

(fn ValidateMaterial_Override
  (Utilities|IsValid (Variables|Default|GetMaterial)
    (:Is%20Valid
      (Rendering|Material|SetMaterial (Variables|Default|GetMainSignBox) 0 (Variables|Default|GetMaterial)))))

(fn ValidateMaterial_TwoSided
  (if (Variables|Default|GetTwoSided)
    (Rendering|Material|SetMaterial (Variables|Default|GetMainSignBox) 1 (Variables|Default|GetMaterial))
    (else
      (Rendering|Material|SetMaterial (Variables|Default|GetMainSignBox) 1 (Variables|Default|GetDummyMaterial)))))

(fn ValidateSignPlaneTransform
  (Transformation|AddLocalOffset (Variables|Default|GetMainSignBox) (Utilities|MainSignBoxOffset) false false)
  (Transformation|SetRelativeScale3D (Variables|Default|GetMainSignBox) (Math|Vector|MakeVector (* (/ (- (.x (Variables|Default|GetBottomRightFlattened)) (.x (Variables|Default|GetTopLeftFlattened))) 1e+02) (* (/ (- 23.0 (- (.x (Variables|Default|GetBottomRightFlattened)) (.x (Variables|Default|GetTopLeftFlattened)))) (- (.x (Variables|Default|GetBottomRightFlattened)) (.x (Variables|Default|GetTopLeftFlattened)))) 0.9975)) 1.0 (* (/ (- (.z (Variables|Default|GetTopLeftFlattened)) (.z (Variables|Default|GetBottomRightFlattened))) 1e+02) (* 0.9975 (/ (- (- (.z (Variables|Default|GetTopLeftFlattened)) (.z (Variables|Default|GetBottomRightFlattened))) 23.0) (- (.z (Variables|Default|GetTopLeftFlattened)) (.z (Variables|Default|GetBottomRightFlattened)))))))))
